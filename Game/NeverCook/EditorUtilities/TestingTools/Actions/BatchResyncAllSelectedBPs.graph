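# Editor utility Blueprint that batch-replaces the selected level actors with freshly spawned instances of their own classes, apparently to resync Blueprint instances.
# ReplaceAllSelected caches the selection, then loops over it using a separate iterator that counts down from an array-length-based start.
# Each step spawns the same class at the original's location and rotation, then destroys the original.
# Scale and per-instance property values are not copied.
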EditorUtilityBlueprint /Game/NeverCook/EditorUtilities/TestingTools/Actions/BatchResyncAllSelectedBPs
graphs [
 EventGraph {links [], nodes []}
 ReplaceAllSelected {nodes [
   ReplaceAllSelected() {is=title K2Node_FunctionEntry}
   > n1 = GetSelectedActors()
   > n2 = Set CahcedArray(CahcedArray:Array of Actor Object References=@n1) -> (Output_Get:Array of Actor Object References)
   n3 = Length(TargetArray=@n2)
   > Set Iterator(Iterator:int=int - int(A=@n3)) -> (Output_Get:int=0)
   > For Loop(LastIndex=@n3)
   n4 = "Get (a copy)"(
    Array:Array of Actor Object References=Get CahcedArray() -> (CahcedArray:Array of Actor Object References)
    Dimension 1=Get Iterator() -> (Iterator:int=0)
   ) -> (Output:Actor Object Reference)
   n5 = Get Actor Transform!(self=@n4) -> (ReturnValue_Location:Vector, ReturnValue_Rotation:Rotator, ReturnValue_Scale:Vector)
   > SpawnActorFromClass!(
    execute
    ActorClass:Actor Class Reference=Get Class(Object=@n4)
    Location:Vector=@n5
    Rotation:Rotator=@n5.ReturnValue_Rotation
    bTransient:bool
   ) -> (then, ReturnValue:Actor Object Reference)
   > Destroy Actor(self=@n4)
   > Decrement Int(Value=Get Iterator() -> (Iterator=0))
  ]}
]
variables []
