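# Editor utility for testing that builds BugItGo camera commands for the selected level actors and copies them to the clipboard.
# For each actor it picks a viewpoint at a random reachable navmesh point near the bounds center, falling back to an extent-based offset, and aims back at the actor.
# Colliding-component bounds take priority over full bounds.
# An embedded Python tkinter script writes the collected commands to the clipboard, one per line.
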EditorUtilityBlueprint /Game/NeverCook/EditorUtilities/TestingTools/Actions/TestingActorActionUtilities
graphs [
 EventGraph {links [], nodes []}
 GetBugItCoordsForSelectedActors {nodes [
   GetBugItCoordsForSelectedActors() {is=title K2Node_FunctionEntry}
   > n1 = GetSelectionSet(self=Get LevelEditorSubsystem())
   > n2 = GetSelectedObjects!(execute, self:Typed Element Selection Set Object Reference=@n1, InRequiredClass:Object Class Reference=&Actor) -> (then, ReturnValue:Array of Object References)
   > n3 = For Each Loop(Array:Array of Object References=@n2) -> (Array Element:Object Reference)
   Execute Python Script!(
    execute=@n3.Completed
    PythonScript="import unreal\r\nfrom tkinter import Tk\r\nimport re\r\n\r\nr = Tk()\r\nr.withdraw()\r\nr.clipboard_clear()\r\n##r.clipboard_append(in_string.replace('\\\\r\\\\n', '\\\\n'))\r\nfor i in range(len(in_strings)):\r\n    append_string = \"\"\r\n    if i > 0:\r\n        append_string = \"\\n\"\r\n    append_string += in_strings[i]\r\n    r.clipboard_append(append_string)\r\nr.destroy()"
    in_strings:Array of Strings=Get l_bugitArray() -> (l_bugitArray:Array of Strings)
   ) -> (then, ReturnValue) {inputs [in_strings]}
   n4 = Sequence(execute=@n3.LoopBody)
   > n5 = Cast To Actor!(execute, Object=@n3) -> (then, CastFailed, AsActor:Actor Object Reference) {is=title K2Node_DynamicCast}
   n6 = Get Actor Bounds(self=@n5, bOnlyCollidingComponents=false)
   n7 = Get Actor Bounds(self=@n5, bOnlyCollidingComponents=true)
   n8 = "Not Equal (Vector)"!(A:Vector=@n7.BoxExtent, B:Vector, ErrorTolerance:float) -> (ReturnValue:bool=~) {is=title K2Node_PromotableOperator}
   > Set l_actorLoc(l_actorLoc:Vector=SelectVector(A=@n7, B=@n6, bPickA=@n8)) -> (Output_Get:Vector)
   n9 = Get Random Reachable Point in Radius!(
    WorldContextObject:Object Reference=@n5
    Origin:"Vector (by ref)"=Get l_actorLoc() -> (l_actorLoc:Vector)
    Radius:float=180
    NavData:Navigation Data Object Reference
    FilterClass:Navigation Query Filter Class Reference
   ) -> (RandomLocation:Vector, ReturnValue:bool) {is=title K2Node_CallFunction}
   > n10 = Branch(Condition=@n9.ReturnValue)
   > Set l_loc(l_loc:Vector=@n9) -> (Output_Get:Vector)
   n11 = SelectVector(A=@n7.BoxExtent, B=@n6.BoxExtent, bPickA=@n8)
   n12 = MakeVector(X=GetMaxElement!(A:Vector=@n11) -> (ReturnValue:double), Y=0)
   Set l_loc(execute=@n10.else, l_loc=vector + vector(A=Get l_actorLoc(), B=vector + vector!(A=@n11, B="50,50,0", C:Vector=@n12) -> (ReturnValue)))
   n13 = BreakVector(InVec=Get l_loc() -> (l_loc:Vector))
   n14 = MakeLiteralString(Value=" ")
   n15 = vector - vector(A=Get l_actorLoc(), B=Get l_loc())
   n16 = BreakRotator(InRot=Rotation From X Vector!(InVec:Vector=vector * vector(A=@n15, B:Vector="1,1,1") -> (ReturnValue=~)) -> (ReturnValue:Rotator))
   n17 = Append!(
    A="BugItGo "
    B=Append!(
     A="To String (Float)"(InDouble=@n13)
     B=@n14
     C="To String (Float)"(InDouble=@n13.Y)
     D:String=@n14
     E:String="To String (Float)"(InDouble=@n13.Z)
     F:String=@n14
    ) -> (ReturnValue)
    C:String=Append!(A="To String (Float)"(InDouble=@n16.Pitch), B=@n14, C="To String (Float)"(InDouble=@n16.Yaw), D=" 0.00") -> (ReturnValue)
   ) -> (ReturnValue)
   n18 = "Equal Exactly (String)"(A=Get l_bugits() -> (l_bugits:String))
   Set l_bugits(
    execute=@n4.then_1
    l_bugits:String=Append!(A=Get l_bugits(), B=Select(Option 0:String="\r\n", Option 1:String, Index=@n18) -> (ReturnValue:String), C=@n17) -> (ReturnValue)
   ) -> (Output_Get:String)
   > Add(TargetArray=Get l_bugitArray(), NewItem=@n17) {is=title K2Node_CallArrayFunction}
   Get Actor Location(self=@n5)
   Execute Python Script!(
    execute
    PythonScript="import unreal\r\nfrom tkinter import Tk\r\n\r\nr = Tk()\r\nr.withdraw()\r\nr.clipboard_clear()\r\nr.clipboard_append(in_string)\r\nr.destroy()"
    in_string:String=Get l_bugits()
   ) -> (then, ReturnValue) {inputs [in_string]}
  ]}
]
variables []
